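# Translucent, two-sided lit particle material for flowing water, named for a temple pipe in the BattlePassS21 quest assets.
# Two sea-foam layers pan at different speeds, their UVs warped by a slow distortion texture, feeding color, opacity and normal-based refraction values.
# Particle color tints the foam beneath a Fresnel highlight, while particle alpha and a V-gradient mask drive erosion.
# Depth fade softens opacity, except on ES2/ES3.1.
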Material /BattlePassS21/QuestActors/Assets/FX/Materials/M_TempleWater_Pipe_02
graph {
 properties {
  bTangentSpaceNormal=false
  bUsedWithMeshParticles=true
  bUsedWithParticleSprites=true
  blendMode=BLEND_Translucent
  numCustomizedUVs=1
  opacityMaskClipValue=0.999
  pixelDepthOffsetMode=PDOM_Legacy
  refractionMethod=RM_IndexOfRefraction
  translucencyLightingMode=TLM_VolumetricPerVertexNonDirectional
  twoSided=true
  usedShadingModels=Default Lit
 }
 nodes [
  n1 = Constant3Vector() {constant=<rgba 0 0 1 0>}
  n2 = Panner(Coordinate=TextureCoordinate() {uTiling=0.5, vTiling=0.5}) {speedY=0.05}
  n3 = Multiply(A=ComponentMask(None=TextureSample(UVs=@n2) {texture=/Game/Global/Textures/Default/Tileable_Patterns/T_Tilling_Distortion_Map_01}.G) {g=true, r=true}) {constB=0.1}
  n4 = Panner(Coordinate=Add(A=@n3, B=TextureCoordinate())) {speedX=0.4, speedY=1.6}
  n5 = TextureSample(UVs=@n4) {samplerType=SAMPLERTYPE_Normal, texture=/Game/Packages/Fortress_Transylvania/Haunted_Forest/Swamps/Textures/T_sea_foam_03_n}
  n6 = Panner(Coordinate=Add(A=@n3, B=TextureCoordinate())) {speedX=-0.2, speedY=1.8}
  n7 = TextureSample(UVs=@n6) {samplerType=SAMPLERTYPE_Normal, texture=&T_sea_foam_03_n}
  n8 = ParticleColor()
  n9 = MaterialFunctionCall() {materialFunction=&LinearGradient}
  n10 = Power(Base=@n9.VGradient) {constExponent=3}
  n11 = Power(Base=OneMinus(None=@n9.VGradient)) {constExponent=1}
  n12 = Multiply(A=Clamp(None=Multiply(A=@n10) {constB=5}), B=Clamp(None=Multiply(A=@n11) {constB=5}))
  n13 = Clamp(None=OneMinus(None=@n12))
  n14 = LinearInterpolate(Alpha=@n8.A) {constB=200}
  n15 = LinearInterpolate(A=LinearInterpolate(A=@n1, B=Add(A=@n5.RGB, B=@n7.RGB), Alpha=@n8.A), B=@n1, Alpha=Clamp(None=Power(Base=@n13, Exp=@n14) {constExponent=35})) {constA=1}
  Add(A=Multiply(A=Clamp(None=@n15)) {constB=0.09}, B=Constant() {r=1})
  n16 = TextureSample(UVs=@n4) {texture=/Game/Packages/Fortress_Transylvania/Haunted_Forest/Swamps/Textures/T_sea_foam_03}
  n17 = TextureSample(UVs=@n6) {texture=&T_sea_foam_03}
  n18 = Multiply(A=@n8.A, B=Clamp(None=@n12))
  n19 = Power(Base=Add(A=@n16.G, B=Multiply(A=@n17.G) {constB=0.4}), Exp=LinearInterpolate(Alpha=@n18) {constA=5, constAlpha=1}) {constExponent=500}
  n20 = LinearInterpolate(Alpha=@n18) {constAlpha=0, constB=2}
  n21 = Multiply(A=Add(A=@n16.A, B=Multiply(A=@n17.A) {constB=0.5}), B=Clamp(None=Multiply(A=@n19, B=@n20) {constB=5}))
  n22 = DepthFade(Opacity=@n21) {fadeDistanceDefault=30}
  FeatureLevelSwitch(Default=@n22, ES2=@n21, ES3_1=@n21, SM4_REMOVED=@n22, SM5=@n22, SM6=@n22)
  n23 = StaticBool() {value=true}
  n24 = Multiply(A=Power(Base=MaterialFunctionCall(Use Cheap contrast=@n23) {materialFunction=&Fresnel_Function}.Result) {constExponent=5}) {constB=30}
  n25 = LinearInterpolate(A=Add(A=@n16.G, B=@n17.G), B=@n16.G, Alpha=@n16.G) {constA=5, constAlpha=1}
  MaterialFunctionCall(DiffuseColor=Add(A=Clamp(None=@n24), B=Multiply(A=@n8.RGB, B=@n25))) {materialFunction=&ConvertFromDiffSpec}
 ]
}
