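# Translucent lit particle material for water effects, apparently pipe outflow in the Battle Pass S21 temple quest set.
# A slowly panning distortion map warps a splash texture and fast-scrolling tornado noise, blended through dynamic parameters to shape color and depth-faded opacity, tinted by particle color.
# Masked splash and distortion normals drive an index-of-refraction value scaled by the dynamic parameter named VTiling.
# A static switch optionally applies time-of-day color tints.
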Material /BattlePassS21/QuestActors/Assets/FX/Materials/M_TempleWater_Pipe_04b
graph {
 properties {
  bTangentSpaceNormal=false
  bUsedWithMeshParticles=true
  bUsedWithParticleSprites=true
  blendMode=BLEND_Translucent
  opacityMaskClipValue=0.933
  pixelDepthOffsetMode=PDOM_Legacy
  refractionMethod=RM_IndexOfRefraction
  translucencyLightingMode=TLM_VolumetricPerVertexNonDirectional
  translucentShadowDensityScale=0
  usedShadingModels=Default Lit
 }
 nodes [
  n1 = Panner(Coordinate=TextureCoordinate() {uTiling=0.5, vTiling=0.5}) {speedY=-0.08}
  n2 = ComponentMask(None=TextureSample(UVs=@n1) {texture=/Game/Global/Textures/Default/Tileable_Patterns/T_Tilling_Distortion_Map_01}.R) {g=true, r=true}
  n3 = Add(A=Multiply(A=@n2) {constB=0.025}, B=TextureCoordinate())
  n4 = Multiply(
   A=TextureSample(UVs=@n3) {samplerType=SAMPLERTYPE_Normal, texture=/Game/Effects/Fort_Effects/Textures/Water/T_water_splash_04_n}.RGB
   B=Constant3Vector() {constant=<rgba 10 10 1 0>}
  )
  n5 = Multiply(
   A=TextureSample(UVs=@n1) {samplerType=SAMPLERTYPE_Normal, texture=/Game/Global/Textures/Default/Tileable_Patterns/T_Tilling_Distortion_Map_01_n}.RGB
   B=Constant3Vector() {constant=<rgba 2 2 1 0>}
  )
  n6 = TextureSample(UVs=@n3) {texture=/Game/Effects/Fort_Effects/Textures/Water/T_water_splash_04}
  n7 = Multiply(A=@n2) {constB=0.05}
  n8 = Panner(Coordinate=Multiply(A=TextureCoordinate(), B=Constant2Vector() {g=3, r=3})) {speedY=-1}
  n9 = DynamicParameter() {defaultValue=<rgba 0 0 0 1>, paramNames [Param1,Param2,VTiling,UTiling]}
  n10 = Power(Base=TextureSample(UVs=Add(A=@n7, B=@n8)) {texture=/Game/Global/Textures/Default/Tileable_Patterns/T_Tornado_Tile}.R, Exp=@n9.Param2)
  n11 = Multiply(A=@n6.A) {constB=80}
  n12 = LinearInterpolate(A=@n6.A, B=Multiply(A=@n10, B=Clamp(None=@n11)), Alpha=Multiply(A=@n9.Param1) {constB=3})
  n13 = Multiply(A=LinearInterpolate(A=Constant3Vector() {constant=<rgba 0 0 1 0>}, B=Add(A=@n4, B=@n5), Alpha=Clamp(None=@n12))) {constB=10}
  Add(A=Multiply(A=Clamp(None=@n13), B=@n9.VTiling) {constB=0.1}, B=Constant() {r=1})
  n14 = Multiply(A=@n6.A, B=@n10)
  n15 = ParticleColor()
  DepthFade(
   Opacity=Multiply(A=LinearInterpolate(A=@n6.A, B=@n14, Alpha=@n9.Param1), B=@n15.A)
   FadeDistance=ScalarParameter() {defaultValue=30, parameterName=Depth Fade}
  ) {fadeDistanceDefault=30}
  n16 = Multiply(A=@n6.R, B=@n10)
  n17 = MaterialFunctionCall(DiffuseColor=Multiply(A=LinearInterpolate(A=@n6.R, B=@n16, Alpha=@n9.Param1), B=@n15.RGB)) {materialFunction=&ConvertFromDiffSpec}
  n18 = Multiply(A=Constant3Vector() {constant=<rgba 0.8148 1 0.9734 0>}, B=Multiply(A=@n17.BaseColor) {constB=0.5})
  n19 = Constant3Vector() {constant=<rgba 0.8148 1 0.9734 0>}
  n20 = Multiply(A=Constant3Vector() {constant=<rgba 0.6927 0.8264 1 0>}, B=@n17.BaseColor)
  StaticSwitchParameter(
   True=MaterialFunctionCall!(Color_Morning=@n18, Color_Day=Multiply(A=@n19, B=@n17.BaseColor), Color_Evening=@n18, Color_Night=Multiply(A=@n20) {constB=0.7}) {materialFunction=/BattlePassS21/QuestActors/Assets/FX/Materials/_Functions/Color_Time_of_Day}.Result
   False=@n17.BaseColor
  ) {parameterName=Use_Color_Time_of_Day}
 ]
}
